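; Base actor for an FM jam session, linking music events to Niagara effects and managing jam membership and audio.
; Gameplay event router beats push game time, and music manager tempo is written, to effects from a getter left empty here.
; Inner collider overlaps toggle server or client membership tags; outer handlers are empty.
; Sound virtualization toggles submix analysis, shutdown destroys non-persistent actors on authority, and EndPlay removes the music event.
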(event EventBeginPlay
  (bind _returnvalue (Class|JamSessionInterface|GetGameplayEventRouterComponent))
  (CallFunction|BindColliderOverlapEvents)
  ; Bind to the Event Router
  (Utilities|FlowControl|Sequence
    (:then_0
      (bind (_asynctaskproxy _payload) (GameplayEventRouter|StartListeningToEvent _returnvalue "/Script/SparksCoreRuntime.SparksCoreAudioEvent_OnBeat")
        (:OnEventReceived
          (CallFunction|OnBeat (Utilities|Struct|BreakSparksCoreAudioEventonBeat _payload) (.BeatInBar (Utilities|Struct|BreakSparksCoreAudioEventonBeat _payload))))))
    (:then_1
      (GameplayEventRouter|StartListeningToEvent _returnvalue "/Script/SparksCoreRuntime.SparksCoreAudioEvent_TempoChanged"
        (:then
          _)
        (:OnEventReceived
          _))
      (CallFunction|OnTempoChanged))))

(event Custom|OnBeat
  (params (Beat%20Number :type Integer) (Beat%20In%20Bar :type Integer))
  (Niagara|SetNiagaraVariable(Float) (CallFunction|GetBeatTempoReactiveNSEffects) "GameTime" (Utilities|Time|GetGameTimeinSeconds)))

(event Custom|InnerColliderBeganOverlap
  (params (OverlappedComponent :type Primitive%20Component%20Object%20Reference) (OtherActor :type Actor%20Object%20Reference) (OtherComp :type Primitive%20Component%20Object%20Reference) (OtherBodyIndex :type Integer) (bFromSweep :type Boolean) (SweepResult :type Hit%20Result%20Structure%20%28by%20ref%29))
  ; Collider Overlap Handling
  (Ability|GameplayEffect|AddLooseGameplayTags OtherActor (select (Networking|HasAuthority) (GameplayTags|MakeLiteralGameplayTagContainer "(GameplayTags=((TagName=\"FMJam.Session.InnerCollider.Member.Server\")))") (GameplayTags|MakeLiteralGameplayTagContainer "(GameplayTags=((TagName=\"FMJam.Session.InnerCollider.Member.Client\")))"))))

(event Custom|OuterColliderBeganOverlap
  (params (OverlappedComponent :type Primitive%20Component%20Object%20Reference) (OtherActor :type Actor%20Object%20Reference) (OtherComp :type Primitive%20Component%20Object%20Reference) (OtherBodyIndex :type Integer) (bFromSweep :type Boolean) (SweepResult :type Hit%20Result%20Structure%20%28by%20ref%29)))

(event Custom|OuterColliderEndedOverlap
  (params (OverlappedComponent :type Primitive%20Component%20Object%20Reference) (OtherActor :type Actor%20Object%20Reference) (OtherComp :type Primitive%20Component%20Object%20Reference) (OtherBodyIndex :type Integer)))

(event Custom|InnerColliderEndedOverlap
  (params (OverlappedComponent :type Primitive%20Component%20Object%20Reference) (OtherActor :type Actor%20Object%20Reference) (OtherComp :type Primitive%20Component%20Object%20Reference) (OtherBodyIndex :type Integer))
  ; Collider Overlap Handling
  (Ability|GameplayEffect|RemoveLooseGameplayTags OtherActor (select (Networking|HasAuthority) (GameplayTags|MakeLiteralGameplayTagContainer "(GameplayTags=((TagName=\"FMJam.Session.InnerCollider.Member.Server\")))") (GameplayTags|MakeLiteralGameplayTagContainer "(GameplayTags=((TagName=\"FMJam.Session.InnerCollider.Member.Client\")))"))))

(event Custom|OnTempoChanged
  (bind _musicmanager (Variables|JamSessionActor|GetMusicManager))
  (bind (_outmusickey _outmusickeymode _outtempo _returnvalue) (Mojo|JamSession|TryGetKeyModeTempo _musicmanager))
  (if _returnvalue
    (Niagara|SetNiagaraVariable(Int32) (CallFunction|GetBeatTempoReactiveNSEffects) "Tempo" _outtempo)))

(event EventBeginShutdownJam
  (Networking|SwitchHasAuthority
    (:Authority
      (Variables|JamSessionActor|GetbIsPersistent
        (:then)
        (:else
          (Actor|DestroyActor))))))

(event OnAudioVirtualizationChanged(JamSessionSoundSource)
  (params (bIsVirtualized :type Boolean))
  (if bIsVirtualized
    (AudioAnalysis|UnRegisterDefaultSubmix (WorldSubsystems|GetFortAudioAnalysisSubsystem))
    (else
      (AudioAnalysis|RegisterDefaultSubmix (WorldSubsystems|GetFortAudioAnalysisSubsystem)))))

(event EventEndPlay
  (params (EndPlayReason :type EEndPlayReason%20Enum :default "Destroyed"))
  ;[0 1 0.6062] Make sure that this even gets removed when we die.  It will also get removed when we deactivate the sound source
  (Audio|RemoveEvent (GameInstanceSubsystems|GetMusicEventSubsystem) (Variables|Default|GetJamSessionSoundSource)))

(fn BindColliderOverlapEvents
  (LocalVariables|SetInner%20Collider (Class|JamSessionInterface|GetInnerJamCollider))
  (LocalVariables|SetOuter%20Collider (Class|JamSessionInterface|GetOuterJamCollider))
  (Utilities|FlowControl|Sequence
    (:then_0
      (LocalVariables|GetInner%20Collider
        (:then
          (Collision|BindEventtoOnComponentBeginOverlap (LocalVariables|GetInner%20Collider) (EventDispatchers|CreateEvent :!event InnerColliderBeganOverlap))
          (Collision|BindEventtoOnComponentEndOverlap (LocalVariables|GetInner%20Collider) (EventDispatchers|CreateEvent :!event InnerColliderEndedOverlap)))
        (:else)))
    (:then_1
      (LocalVariables|GetOuter%20Collider
        (:then
          (Collision|BindEventtoOnComponentBeginOverlap (LocalVariables|GetOuter%20Collider) (EventDispatchers|CreateEvent :!event OuterColliderBeganOverlap))
          (Collision|BindEventtoOnComponentEndOverlap (LocalVariables|GetOuter%20Collider) (EventDispatchers|CreateEvent :!event OuterColliderEndedOverlap)))
        (:else)))))

(fn GetBeatTempoReactiveNSEffects
  (return 0))

(fn GetInnerCollider
  (return 0))

(fn GetJamAudioSource
  (return (Variables|Default|GetJamSessionSoundSource)))

(fn GetOuterCollider
  (return 0))

(fn ConstructionScript)
