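; BP_JamSessionActor_Rectangular is the rectangular variant of a jam session actor.
; It overrides getters to supply its own inner and outer colliders, plus an array holding the square inner border Niagara effect as the beat-tempo-reactive effect.
; The construction script only calls the parent, so this file adds no session logic beyond these shape-specific components.
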(fn GetBeatTempoReactiveNSEffects
  (return (Utilities|Array|MakeArray (Variables|Default|GetNS_JamSpace_Border_SquareInner))))

(fn GetInnerCollider
  (return (Variables|Default|GetInnerCollider)))

(fn GetOuterCollider
  (return (Variables|Default|GetOuterCollider)))

(fn ConstructionScript
  (|CallParentFunction))
